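; Gameplay ability for the Voyager flying orb shield that knocks the avatar pawn away and applies GE_OrbRedeploy to allow glider redeployment.
; DoLaunch forces custom movement such as wingsuit to falling, then launches with velocity from AdjustLaunchVector and prints debug info.
; AdjustLaunchVector boosts horizontal force with proximity to the granting effect's causer, damping all axes for custom-movement or skydiving pawns.
; Activation adds no logic here, and DoLaunch has no caller in this file.
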(event Ability|EventActivateAbility
  (|CallParentFunction) ; This node is disabled and will not be called.\nDrag off pins to build functionality.
)

(event Ability|EventOnEndAbility
  (params (bWasCancelled :type Boolean)))

(fn AdjustLaunchVector
  (params (Vel :type Vector) (Horz%20Distance :type Float%20%28double-precision%29))
  ;[0 0.15 0.0652 0.5] Adjust knockback force depending on movement modes
  (return (Math|Vector|MakeVector (* (.x Vel) (Math|Float|SelectFloat (/ (Variables|Default|GetKnockbackRadius) (Math|Float|Clamp(Float) Horz%20Distance 0.0 (Variables|Default|GetKnockbackRadius))) (/ (/ (Variables|Default|GetKnockbackRadius) (Math|Float|Clamp(Float) Horz%20Distance 0.0 (Variables|Default|GetKnockbackRadius))) (Variables|Default|GetBounceVelAdjustValue)) (or (Variables|Default|GetUsedCustomMovement) (GameplayTags|HasAnyMatchingGameplayTags (Ability|GetFortAbilitySystemComponent (Variables|Default|GetKnockedPawn)) (Variables|Default|GetSkydiveTag))))) (* (.y Vel) (Math|Float|SelectFloat (/ (Variables|Default|GetKnockbackRadius) (Math|Float|Clamp(Float) Horz%20Distance 0.0 (Variables|Default|GetKnockbackRadius))) (/ (/ (Variables|Default|GetKnockbackRadius) (Math|Float|Clamp(Float) Horz%20Distance 0.0 (Variables|Default|GetKnockbackRadius))) (Variables|Default|GetBounceVelAdjustValue)) (or (Variables|Default|GetUsedCustomMovement) (GameplayTags|HasAnyMatchingGameplayTags (Ability|GetFortAbilitySystemComponent (Variables|Default|GetKnockedPawn)) (Variables|Default|GetSkydiveTag))))) (Math|Float|SelectFloat (/ (.z Vel) (Variables|Default|GetBounceVelAdjustValue)) (.z Vel) (or (Variables|Default|GetUsedCustomMovement) (GameplayTags|HasAnyMatchingGameplayTags (Ability|GetFortAbilitySystemComponent (Variables|Default|GetKnockedPawn)) (Variables|Default|GetSkydiveTag)))))))

(fn DoLaunch
  (params (Vel :type Vector :default "0, 0, 0"))
  (bind _knockbackradius (Variables|Default|GetKnockbackRadius))
  (bind _returnvalue (Default|GetAbilityInstigator))
  (bind _returnvalue_2 (Transformation|GetHorizontalDistanceTo _returnvalue (Ability|GetAvatarActorfromActorInfo)))
  (bind _returnvalue_3 (Math|Float|Clamp(Float) _returnvalue_2 0.0 _knockbackradius))
  (bind _adjustedvel (CallFunction|AdjustLaunchVector Vel _returnvalue_3))
  (bind _returnvalue_4 (Ability|GetAvatarActorfromActorInfo))
  (bind _knockedpawn (Variables|Default|GetKnockedPawn))
  (bind _charactermovement (Variables|/Script/Engine.Character|GetCharacterMovement _knockedpawn))
  ;[0 0.15 0.0652 0.5] Launch players knocked back by orb
  (Variables|Default|SetKnockedPawn (Utilities|Casting|CastToFortPlayerPawnAthena _returnvalue_4))
  (if (Utilities|Enum|Equal(Enum) (Variables|/Script/Engine.CharacterMovementComponent|GetMovementMode _charactermovement) "MOVE_Custom")
    ;;[0 0.15 0.0652 0.9478] Force player out of wingsuit mode.
    (Variables|Default|SetUsedCustomMovement true) ; Use
    (Pawn|Components|CharacterMovement|SetMovementMode _charactermovement "MOVE_Falling" 0))
  (Character|LaunchCharacter (Ability|GetActivatingPawn) _adjustedvel true true)
  (Development|PrintString (Utilities|String|Append "Launch Velo is " (Utilities|String|ToString(Vector) _adjustedvel) " and the player is this far away: " (Utilities|String|ToString(Float) _returnvalue_3) ". Custom Movement was: " (Utilities|String|ToString(Boolean) (Variables|Default|GetUsedCustomMovement))))
  (GameplayEffects|ApplyGameplayEffectToSelf :self (Ability|GetFortAbilitySystemComponent (Variables|Default|GetKnockedPawn)) :GameplayEffectClass "/DragonCartNarrative/Gameplay/FlyingOrb/Effects/GE_OrbRedeploy.GE_OrbRedeploy_C" :Level 0.0) ; Apply ability to allow for glider redeployment.
)

(fn GetAbilityInstigator
  (return (Ability|EffectContext|GetEffectCauser (Ability|GetGrantedbyEffectContext))))
